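; Weapon bunker door for the Ch6S4 mid-match bunkers, a child Blueprint that opens on storm phase with military-rank unlocking disabled.
; Opening sets a priming state, then a timer makes the server slide both doors by timeline and mark the bunker for rank-tagged players.
; The replicated state's OnRep updates map icons, drives klaxon, smoke and audio cues, and snaps doors open for late clients.
; An overlap query finds the entrance frame to tint its lights by rank.
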(event EventBeginPlay
  (bind _returnvalue (Fortnite|GetFirstLocalFortPlayerController true))
  ;[0 0 0] Begin Play
  (|CallParentFunction)
  (Utilities|FlowControl|Sequence
    (:then_0
      (Variables|Default|SetCollisionBlockerEnable false)
      (Variables|Default|SetDoorRight_Start_RelativeLoc (Variables|/Script/Engine.SceneComponent|GetRelativeLocation (Variables|Default|GetDoorRight)))
      (Variables|Default|SetDoorLeft_Start_RelativeLoc (Variables|/Script/Engine.SceneComponent|GetRelativeLocation (Variables|Default|GetDoorLeft)))
      (CallFunction|OverrideMapIcon) ; Set Map icons to match S6CH4 Bunker
    )
    (:then_2
      (Utilities|IsValid _returnvalue
        (:Is%20Valid
          (CallFunction|OnLocalPawnControllerChanged 0 _returnvalue))
        (:Is%20Not%20Valid
          (EventDispatchers|AssignOnPawnControllerChanged (Game|GetGameInstance) (EventDispatchers|CreateEvent :!event Custom|OnLocalPawnControllerChanged)))))
    (:then_3
      (Utilities|Time|SetTimerbyEvent (EventDispatchers|CreateEvent :!event Custom|FindDoorFrame) 2.0))))

(event EventStartOpeningSequence
  ; 37.40 Onward we are opening bunkers on Storm Phase because we are disabling Military Ranking
  (|CallParentFunction)
  (CallFunction|OpeningDoor))

(event Custom|OnLocalPlayerStateReady
  (params (Controller :type Fort%20Player%20Controller%20Object%20Reference) (PlayerState :type Fort%20Player%20State%20Object%20Reference))
  ;[0.2277 0.2435 0.2448] Pawn Related
  (EventDispatchers|UnbindEventfromOnPlayerStateReady Controller (EventDispatchers|CreateEvent :!event Custom|OnLocalPlayerStateReady)))

(event Custom|FindDoorFrame
  ; Adjust LED color
  (bind (_outactors _returnvalue) (Collision|BoxOverlapActors (Transformation|GetActorLocation) "200.000000,200.000000,200.000000" (Utilities|Array|MakeArray "ObjectTypeQuery1") "/Hermes/Sets/CrewCut/Blueprints/BSG/BSG_CrewCut_Entrance_A.BSG_CrewCut_Entrance_A_C" :!class /Script/Engine.KismetSystemLibrary))
  (if _returnvalue
    (bind _asbsg_crew_cut_entrance_a (Utilities|Casting|CastToBSG_CrewCut_Entrance_A (Utilities|Array|Get(acopy) _outactors)))
    (bind _output_get (Variables|Default|SetBSGEntranceFrame _asbsg_crew_cut_entrance_a))
    ;; Check to set the right color on lights when in and out of relevancy
    (if (GameplayTags|HasMatchingGameplayTag (Game|Player|GetPlayerPawn (Fortnite|GetFirstLocalFortPlayerController true)) (GameplayTags|MakeLiteralGameplayTag "(TagName=\"CherrySmoke.HeroRank.A\")"))
      (Class|BSGCrewCutEntranceA|UpdateLight _output_get "(R=0.084376,G=1.000000,B=0.144128,A=1.000000)")
      (else
        (Class|BSGCrewCutEntranceA|UpdateLight _output_get "(R=1.000000,G=0.003302,B=0.023465,A=1.000000)")))
    (else
      (Fortnite|FortShippingLog self "BGA Bunker Door: Did not find the Entrance BSG on begin play"))))

(event Custom|OnLocalPawnControllerChanged
  (params (Pawn :type Pawn%20Object%20Reference) (Controller :type Controller%20Object%20Reference))
  ;[0.2277 0.2435 0.2448] Pawn Related
  (EventDispatchers|UnbindEventfromOnPawnControllerChanged (Game|GetGameInstance) (EventDispatchers|CreateEvent :!event Custom|OnLocalPawnControllerChanged))
  (Utilities|IsValid (Variables|/Script/Engine.Controller|GetPlayerState Controller)
    (:Is%20Not%20Valid
      (bind _asfort_player_controller (Utilities|Casting|CastToFortPlayerController Controller)
        (:then
          (EventDispatchers|BindEventtoOnPlayerStateReady _asfort_player_controller (EventDispatchers|CreateEvent :!event Custom|OnLocalPlayerStateReady)))
        (:CastFailed
          (Development|PrintString "Warning: Local player controller is not a FortPlayerController"))))))

(event Custom|EnableKlaxxonPulsing
  (params (PulseDuration :type Float%20%28double-precision%29))
  (bind _staticmeshklaxon (Variables|Default|GetStaticMeshKlaxon))
  (bind _material_instance_dynamic (Class|BTechArtGlobalLibrary|GetDynamicMaterialInstance _staticmeshklaxon))
  (bind _returnvalue (Math|Float|Max(Float) PulseDuration 0.1))
  (bind _returnvalue_2 (/ 1.0 _returnvalue))
  ;[0.01402 0.08333 0.05784] Enable or disable Klaxxon pulsing.
  (Networking|CanExecuteCosmeticEvents
    (:True
      (Rendering|SetVisibility (Variables|Default|GetFlare01) true false)
      (Rendering|Material|SetVectorParameterValue _material_instance_dynamic "SFX_Emissive_Tint(Expand)" (Variables|Materials|GetKlaxonOpeningEmissive) :!class /Script/Engine.MaterialInstanceDynamic)
      (Rendering|Material|SetScalarParameterValue _material_instance_dynamic "Emissive Time Speed Layer 1" _returnvalue_2 :!class /Script/Engine.MaterialInstanceDynamic)
      (Rendering|Material|SetScalarParameterValue (Variables|Materials|GetKlaxonLightFunctionMID) "Speed" _returnvalue_2 :!class /Script/Engine.MaterialInstanceDynamic)
      (Rendering|Material|SetScalarParameterValue (Class|BTechArtGlobalLibrary|GetDynamicMaterialInstance (Variables|Default|GetFlare01)) "Speed" _returnvalue_2 :!class /Script/Engine.MaterialInstanceDynamic))))

(event Custom|Opening%20Door
  ;[0.2277 0.2435 0.2448] Door Opening Sequence
  (Utilities|Time|SetTimerbyEvent (delegate Custom|Door%20Opening%20sequence) (Variables|Default|GetPreping%20Time))
  (Variables|Default|SetWeaponBunkerDoorState "NewEnumerator1"))

(event Custom|DisableKlaxxonPulsing
  ;[0.01402 0.08333 0.05784] Enable or disable Klaxxon pulsing.
  (Networking|CanExecuteCosmeticEvents
    (:True
      (Rendering|SetVisibility (Variables|Default|GetPointLight) false false))))

(event Custom|FadeKlaxonLightIn
  ;[0.01402 0.08333 0.05784] Enable or disable Klaxxon pulsing.
  (bind (_direction _fade) (Timeline|FadeKlaxonLightTL PlayFromStart)
    (:Update
      (Rendering|Components|Light|SetIntensity (Variables|Default|GetPointLight) (Math|Float|Lerp 0.0 (Variables|Materials|GetKlaxonLightIntensity) _fade)))))

(event Custom|ChangeLightColor
  (params (NewColor :type Linear%20Color%20Structure))
  ;[0.01402 0.08333 0.05784] Enable or disable Klaxxon pulsing.
  (if (not (Components|Timeline|IsPlaying (Variables|BGA_BunkerDoor_TidalCrane|GetChangeLightColorTL)))
    (Variables|Materials|SetTargetLightColor NewColor)
    (Variables|Materials|SetCurrentLightColor (Rendering|Components|Light|GetLightColor (Variables|Default|GetPointLight)))
    (bind (_direction _lerp) (Timeline|ChangeLightColorTL PlayFromStart)
      (:Update
        (Rendering|Components|Light|SetLightColor (Variables|Default|GetPointLight) (Math|Color|Lerp(LinearColor) (Variables|Materials|GetCurrentLightColor) (Variables|Materials|GetTargetLightColor) _lerp) true)))))

(event Custom|FadeKlaxonLightOut
  ;[0.01402 0.08333 0.05784] Enable or disable Klaxxon pulsing.
  (Timeline|FadeKlaxonLightTL ReverseFromEnd))

(event Custom|Door%20Opening%20sequence
  ;[0.2277 0.2435 0.2448] Door Opening Sequence
  (Variables|Default|SetWeaponBunkerDoorState "NewEnumerator2")
  (CallFunction|UpdateMarkerDisplayInfo) ; Update marker display info based on Weapon Bunker Door State
  (Variables|MapIcons|SetMapIconOpened) ; Update map icon to opened bunnker
  (Networking|SwitchHasAuthority
    (:Authority
      (Utilities|FlowControl|Sequence
        (:then_0
          (Utilities|FlowControl|DoOnce false
            (:Completed
              (Variables|Default|SetCollisionBlockerEnable false)
              (bind (_direction _movecrate) (Timeline|Timeline PlayFromStart :NewTime (Variables|Default|GetDoorOpenDuration))
                (:Update
                  (Transformation|SetRelativeLocation (Variables|Default|GetDoorRight) (Math|Vector|Lerp(Vector) (Variables|Default|GetDoorRight_Start_RelativeLoc) (+ (Variables|Default|GetDoorRight_Start_RelativeLoc) "930.000000,0.000000,0.000000") _movecrate) false false)
                  (Transformation|SetRelativeLocation (Variables|Default|GetDoorLeft) (Math|Vector|Lerp(Vector) (Variables|Default|GetDoorLeft_Start_RelativeLoc) (- (Variables|Default|GetDoorLeft_Start_RelativeLoc) "930.000000,0.000000,0.000000") _movecrate) false false))
                (:Finished
                  (Transformation|SetRelativeTransform (Variables|Default|GetDoorRight) (Math|Transform|MakeTransform (Variables|Default|GetInitRightEndWorldLoc)) false false)
                  (Transformation|SetRelativeTransform (Variables|Default|GetDoorLeft) (Math|Transform|MakeTransform (Variables|Default|GetInitLeftEndWorldLoc) "0.000000,-180.000000,0.000000") false false)
                  (CallFunction|SFXPlayDoorFinishedOpen)
                  (Variables|Default|SetWeaponBunkerDoorState "NewEnumerator5"))))))
        (:then_1
          (for _array_element (Fortnite|GetAllFortPlayerControllers true true)
            (bind _returnvalue (GameplayTags|HasAnyMatchingGameplayTags(Message) (Game|Player|GetPlayerPawn _array_element) (Variables|Default|GetMilitary%20Rank%20Tags)))
            (if _returnvalue
              (Utilities|IsValid _array_element
                (:Is%20Valid
                  (Utilities|IsValid (Pawn|GetViewTarget _array_element)
                    (:Is%20Valid
                      (bind _asfort_player_state_athena (Utilities|Casting|CastToFortPlayerStateAthena (Variables|/Script/Engine.Controller|GetPlayerState _array_element)))
                      (if (<= (Variables|/Script/FortniteGame.FortPlayerStateAthena|GetPlace _asfort_player_state_athena) 0)
                        (Utilities|IsValid (Actor|GetComponentbyClass _array_element "/Script/FortniteGame.AthenaMarkerComponent")
                          (:Is%20Valid
                            (MarkerComponent|MarkActoronServer (Actor|GetComponentbyClass _array_element "/Script/FortniteGame.AthenaMarkerComponent") self false true)))))))))))))
    (:Remote
      (CallFunction|SFXPlayDoorStartOpen))))

(event Custom|MultiX_ActivatePrimingFX
  ;[0.2277 0.2435 0.2448] Bunker Door State Change
  (Networking|SwitchHasAuthority
    (:Remote
      ;; Klaxxon on Prime
      (CallFunction|EnableKlaxxonPulsing 1.5)
      (CallFunction|FadeKlaxonLightIn)
      (CallFunction|ActivatePrimeAudio)
      (Components|Activation|Activate (Variables|Default|GetNiagara_Smoke_Priming1) false))))

(event EventClientWeaponBunkerDoorStateChanged
  (params (NewState :type WeaponBunkerDoorOpeningStateEnum%20Enum :default "NewEnumerator0")))

(event EventOnRep_WeaponBunkerDoorState
  (bind _athenaspecialactor (Variables|Default|GetAthenaSpecialActor))
  (bind _athenaspecialactor_2 (Variables|Default|GetAthenaSpecialActor))
  (bind _athenaspecialactor_3 (Variables|Default|GetAthenaSpecialActor))
  ;[0.2277 0.2435 0.2448] Bunker Door State Change
  (|CallParentFunction)
  (switch Utilities|FlowControl|Switch|SwitchonWeaponBunkerDoorOpeningStateEnum (Variables|Default|GetWeapon%20Bunker%20Door%20State)
    (:NewEnumerator1
      ;;[1 0 0 0.5] TODO - clean up this hack and maybe get rid of this component - could be integrated into the map icon actor instead?
      (Utilities|IsValid _athenaspecialactor_2
        (:Is%20Valid
          (Components|DestroyComponent _athenaspecialactor_2))
        (:Is%20Not%20Valid
          _))
      (Development|PrintString "Bunker is Priming")
      (Variables|Default|SetPartly%20Opened%20Completed true)
      (Networking|SwitchHasAuthority
        (:Authority
          (CallFunction|SetMapIconOpening)
          (CallFunction|MultiXActivatePrimingFX))))
    (:NewEnumerator2
      (Development|PrintString "PlayOpening Animation")
      (Utilities|IsValid _athenaspecialactor_3
        (:Is%20Valid
          (Components|DestroyComponent _athenaspecialactor_3)
          ;; Comment
          (Networking|SwitchHasAuthority
            (:Remote
              (CallFunction|EnableKlaxxonPulsing 1.5))))))
    (:NewEnumerator5
      (Utilities|IsValid _athenaspecialactor
        (:Is%20Valid
          (Components|DestroyComponent _athenaspecialactor))
        (:Is%20Not%20Valid
          _))
      (Transformation|SetRelativeTransform (Variables|Default|GetDoorRight) (Math|Transform|MakeTransform (Variables|Default|GetInitRightEndWorldLoc) (0.0 0.0 0.0)) false false) ; Set the final transform of the static mesh door, \nuseful on the client if they missed the opening state.
      (Transformation|SetRelativeTransform (Variables|Default|GetDoorLeft) (Math|Transform|MakeTransform (Variables|Default|GetInitLeftEndWorldLoc) (0.0 -1.8e+02 0.0)) false false) ; Set the final transform of the static mesh door, \nuseful on the client if they missed the opening state.
      (CallFunction|RefreshMarkerDisplayInfo)
      (Networking|SwitchHasAuthority
        (:Authority
          (Variables|MapIcons|SetMapIconOpened))
        (:Remote
          (CallFunction|EnableKlaxxonPulsing 1.5)
          (CallFunction|DeactivatePrimeAudio))))))

(event Custom|ActivatePrimeAudio
  (bind _audio_priming (Variables|Audio|GetAudio_Priming))
  ; Activate Priming Audio
  (if (not (and (Audio|Components|Audio|IsPlaying _audio_priming) (Variables|Default|GetbDidAudioPrimePlay)))
    (Components|Activation|Activate _audio_priming false)
    (Audio|Components|Audio|Play _audio_priming 0.0)
    (Variables|Default|SetbDidAudioPrimePlay true)
    (SoundIndicator|ShowSoundIndicatorforDuration (Variables|Default|GetFortSoundIndicator) (Transformation|GetActorLocation) (Variables|/Script/Engine.AudioComponent|GetSound _audio_priming) 1.0 15.0 0)
    (Utilities|FlowControl|Delay 9.5)
    (Components|Activation|Activate (Variables|Default|GetNiagara_Smoke) false)))

(event Custom|DeactivatePrimeAudio
  ; Activate Priming Audio
  (Components|Activation|Deactivate (Variables|Audio|GetAudio_Priming)))

(event Custom|SFX_Play_DoorStartOpen
  (Variables|Default|GetDoor%20Opened
    (:then)
    (:else
      (Utilities|FlowControl|Sequence
        (:then_0
          (bind _returnvalue (Audio|SpawnSoundAttached "/MidMatchWeaponBunkerGameplay/Items/WeaponBunkerDoor/Audio/MSS_Bunker_Doors_Movement.MSS_Bunker_Doors_Movement" (Variables|Audio|GetAudioSpawnPoint_DoorLeft)))
          (bind _output_get (Variables|Audio|SetAudio_DoorLeft _returnvalue))
          (Audio|Parameter|SetObjectArrayParameter _output_get "Wavs" (Variables|Audio|GetAudio_DoorLeft_Wavs)))
        (:then_1
          (bind _returnvalue_2 (Audio|SpawnSoundAttached "/MidMatchWeaponBunkerGameplay/Items/WeaponBunkerDoor/Audio/MSS_Bunker_Doors_Movement.MSS_Bunker_Doors_Movement" (Variables|Audio|GetAudioSpawnPoint_DoorRight)))
          (bind _output_get_2 (Variables|Audio|SetAudio_DoorRight _returnvalue_2))
          (Audio|Parameter|SetObjectArrayParameter _output_get_2 "Wavs" (Variables|Audio|GetAudio_DoorRight_Wavs))))
      (Audio|Parameter|SetFloatParameter (Variables|Audio|GetAudio_DoorRight) "Duration" (Variables|Default|GetDoorOpenDuration) :!class /Script/AudioExtensions.AudioParameterControllerInterface)
      (Audio|Parameter|ExecuteTriggerParameter (Variables|Audio|GetAudio_DoorLeft) "Open.Start")
      (Utilities|Time|SetTimerbyEvent (delegate Custom|Audio_DoorOpenSafetyCheck) (+ (Variables|Default|GetDoorOpenDuration) 1.0))
      (Variables|Default|SetDoor%20Opened true)
      (Audio|Components|Audio|Play (Variables|Audio|GetAudio_Bunker_Music) 0.0))))

(event Custom|Audio_DoorOpenSafetyCheck
  (Utilities|FlowControl|Sequence
    (:then_0
      (bind _audio_doorleft (Variables|Audio|GetAudio_DoorLeft)
        (:then
          (Audio|Components|Audio|FadeOut _audio_doorleft 0.7 0.0 "Linear" :!class /Script/Engine.AudioComponent))
        (:else)))
    (:then_1
      (bind _audio_doorright (Variables|Audio|GetAudio_DoorRight)
        (:then
          (Audio|Components|Audio|FadeOut _audio_doorright 0.7 0.0 "Linear" :!class /Script/Engine.AudioComponent))
        (:else)))))

(event Custom|SFX_Play_DoorFinishedOpen
  (Networking|CanExecuteCosmeticEvents
    (:True
      (Audio|Parameter|ExecuteTriggerParameter (Variables|Audio|GetAudio_DoorRight) "Open.Finish"))))

(fn CanLocalPlayerUnlockDoor
  (params (RankChangedPayload :type On%20Hero%20Rank%20Changed%20Structure) (HeroRankToCheck :type Gameplay%20Tag%20Structure))
  ; Removing Military Ranking from Bunkers
  (return false))

(fn CreateMapIcon
  (bind _mapiconactor (|CallParentFunction))
  (bind _output_get (Variables|Default|SetMap%20Icon%20Actor _mapiconactor))
  (return _output_get))

(fn GetMarkerDetailsId
  (return "None"))

(fn OnRep_CollisionBlockerEnable
  (bind _doorblockercollision (Variables|Default|GetDoorBlockerCollision))
  (if (Variables|Default|GetCollisionBlockerEnable)
    (Collision|SetCollisionEnabled _doorblockercollision "QueryAndPhysics")
    (else
      (Collision|SetCollisionEnabled _doorblockercollision "NoCollision"))))

(fn OverrideMapIcon
  (bind _map_icon_actor (Variables|Default|GetMap%20Icon%20Actor))
  (bind _athenaspecialactor_opening (Variables|/MidMatchWeaponBunkerGameplay/Items/WeaponBunkerDoor/B_WeaponBunkerMapIconActor.B_WeaponBunkerMapIconActor_C|GetAthenaSpecialActor_Opening _map_icon_actor))
  (bind _map_icon_actor_2 (Variables|Default|GetMap%20Icon%20Actor))
  (bind _athenaspecialactor_opened (Variables|/MidMatchWeaponBunkerGameplay/Items/WeaponBunkerDoor/B_WeaponBunkerMapIconActor.B_WeaponBunkerMapIconActor_C|GetAthenaSpecialActor_Opened _map_icon_actor_2))
  (Utilities|IsValid (Variables|Default|GetMap%20Icon%20Actor)
    (:Is%20Valid
      _)
    (:Is%20Not%20Valid
      (bind _mapiconactor (CallFunction|CreateMapIcon))
      (Variables|Default|SetMap%20Icon%20Actor _mapiconactor)))
  (Utilities|FlowControl|Sequence
    (:then_0
      (Class|AthenaSpecialActorComponent|SetIcons _athenaspecialactor_opening (Variables|MapIcons|GetMapIcon_Opening) (Variables|MapIcons|GetMapIcon_Opening) (Variables|MapIcons|GetMapIcon_Opening)) ; Changing the icon for the opening state
      (Components|Activation|Deactivate _athenaspecialactor_opening))
    (:then_1
      (Class|AthenaSpecialActorComponent|SetIcons _athenaspecialactor_opened (Variables|MapIcons|GetMapIcon_Opened) (Variables|MapIcons|GetMapIcon_Opened) (Variables|MapIcons|GetMapIcon_Opened)) ; Changing the icon for the opened state
      (Components|Activation|Deactivate _athenaspecialactor_opened))))

(fn RefreshMarkerDisplayInfo
  (|CallParentFunction))

(fn SetMapIconOpened
  (|CallParentFunction)
  (Class|AthenaSpecialActorComponent|SetIcons (Variables|/MidMatchWeaponBunkerGameplay/Items/WeaponBunkerDoor/B_WeaponBunkerMapIconActor.B_WeaponBunkerMapIconActor_C|GetAthenaSpecialActor_Opened (Variables|Default|GetMap%20Icon%20Actor)) (Variables|MapIcons|GetMapIcon_Opened) (Variables|MapIcons|GetMapIcon_Opened) (Variables|MapIcons|GetMapIcon_Opened)))

(fn SetWeaponBunkerDoorState
  (params (NewState :type WeaponBunkerDoorOpeningStateEnum%20Enum :default "NewEnumerator0"))
  (bind _output_get (Variables|Default|SetWeapon%20Bunker%20Door%20State NewState))
  (|CallParentFunction _output_get))

(fn UpdateMarkerDisplayInfo
  (bind _weaponbunkermarkedactordisplayinfodatatable (Variables|General|GetWeaponBunkerMarkedActorDisplayInfoDatatable)
    (:then
      (bind _returnvalue (Utilities|GetDataTableRow _weaponbunkermarkedactordisplayinfodatatable (Utilities|Select :Index (Variables|Default|GetWeapon%20Bunker%20Door%20State) :NewEnumerator0 "MarkerDetails.MMObj.WeaponBunkerGameplay" :NewEnumerator6 "MarkerDetails.MMObj.WeaponBunkerGameplay" :NewEnumerator1 "MarkerDetails.MMObj.WeaponBunkerGameplay.Open" :NewEnumerator2 "MarkerDetails.MMObj.WeaponBunkerGameplay.Open" :NewEnumerator5 "MarkerDetails.MMObj.WeaponBunkerGameplay.Open")))
      (Variables|Marking|SetMarkerDisplay _returnvalue))
    (:else)))

(fn ConstructionScript
  (|CallParentFunction))
