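; Parent Blueprint for weapon bunker doors, running a server-authoritative, replicated door state from closed through priming to opening.
; Once playlist data is ready, the server closes the door and spawns its map icon actor.
; A one-shot StartOpeningSequence, called from the objective's game state component, primes, waits a hotfixable delay, then opens.
; State changes flush net dormancy and fire a client event for bursts; helpers update map icon and marker info.
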(event EventBeginPlay
  ;[0.06801 0.5677 0.1422] BeginPlay: At the start of the match, the bunker door state is closed.
  (|CallParentFunction)
  (Networking|SwitchHasAuthority
    (:Authority
      (Class|FortAsyncActionCurrentPlaylistReady|PlaylistDataReadyAsync
        (:OnReady
          (CallFunction|SetWeaponBunkerDoorState "NewEnumerator6")
          (CallFunction|CreateMapIcon))))))

(event Custom|StartOpeningSequence
  ;[0.2083 0.04968 0.1819] StartOpeningSequence: Switches the bunker door to the priming state and after the desired amount of delay, switches it to the Opening state to start opening the door.
  ; Called from the objective's game state component.
  (Networking|SwitchHasAuthority
    (:Authority
      (Utilities|FlowControl|DoOnce false
        (:Completed
          (CallFunction|SetWeaponBunkerDoorState "NewEnumerator1")
          (Utilities|FlowControl|Delay (CallFunction|GetPrimingDuration))
          (CallFunction|SetWeaponBunkerDoorState "NewEnumerator2"))))))

(event Custom|OnRep_WeaponBunkerDoorState
  ;[0.6805 0.3014 1] OnRep_WeaponBunkerDoorOpeningState: Handle the various states of the weapon bunker door actor.
  ; Called on both server and client.
  ;;[0.214 0.5449 1] Log the new state on the client and server.
  (Fortnite|FortShippingLog self (Utilities|String|Append "OnRep_WeaponBunkerDoorState: State changed to : " (Utilities|String|EnumtoString (Variables|Transient|GetWeaponBunkerDoorState)))))

(event Custom|ClientWeaponBunkerDoorStateChanged
  (params (NewState :type WeaponBunkerDoorOpeningStateEnum%20Enum :default "NewEnumerator0")))

(event Custom|OnRep_bMarkTraceResponseEnabled)

(event Custom|OnRep_bPrimingEnabled)

(fn CreateMapIcon
  ;[0.02925 0.1838 0.1875] CreateMapIcon: Create the map icon actor if it doesn't exist.
  (Networking|SwitchHasAuthority
    (:Authority
      (Variables|Transient|GetWeaponBunkerMapIcon
        (:then
          _)
        (:else
          (bind _returnvalue (Game|SpawnActorfromClass "/MidMatchWeaponBunkerGameplay/Items/WeaponBunkerDoor/B_WeaponBunkerMapIconActor.B_WeaponBunkerMapIconActor_C" (Transformation|GetActorTransform) "AlwaysSpawn" "MultiplyWithRoot" 0 0))
          (Variables|Transient|SetWeaponBunkerMapIcon _returnvalue)))
      (return (Variables|Transient|GetWeaponBunkerMapIcon)))))

(fn DestroyMapIcon
  ;[0.0285 0.1615 0.1566] DestroyMapIcon: Destroy the spawned map icon actor.
  (Networking|SwitchHasAuthority
    (:Authority
      (bind _weaponbunkermapicon (Variables|Transient|GetWeaponBunkerMapIcon)
        (:then
          (Actor|DestroyActor _weaponbunkermapicon)
          (Variables|Transient|SetWeaponBunkerMapIcon 0))
        (:else)))))

(fn GetMarkerDetailsId
  (return (select (Utilities|Enum|Equal(Enum) (Variables|Transient|GetWeaponBunkerDoorState) (LocalVariables|GetOpenedState)) "MarkerDetails.MMObj.WeaponBunkerGameplay.Open" "MarkerDetails.MMObj.WeaponBunkerGameplay")))

(fn GetPrimingDuration
  (return (Math|Float|Max(Float) (ScalableFloat|GetValueatLevel (Variables|Hotfix|GetHotfix_PrimingDuration)) 1.0)))

(fn HandleAnalyticsCollectionDurationComplete)

(fn RefreshMarkerDisplayInfo
  ;[0.0151 0.1615 0.1582] RefreshMarkerDisplayInfo: Update the marker display info based on the marker details id.
  (bind _weaponbunkermarkedactordisplayinfodatatable (Variables|General|GetWeaponBunkerMarkedActorDisplayInfoDatatable)
    (:then
      (bind _returnvalue (Utilities|GetDataTableRow _weaponbunkermarkedactordisplayinfodatatable (MarkableActor|GetMarkerDetailsId)))
      (Variables|Marking|SetMarkerDisplay _returnvalue))
    (:else)))

(fn SetMapIconOpened
  ;[0.1161 0.1562 0.1419] SetMapIconOpened: Create the map icon actor, if it doesn't exist, and set the map icon.
  (Networking|SwitchHasAuthority
    (:Authority
      (bind _mapiconactor (CallFunction|CreateMapIcon))
      (Utilities|IsValid _mapiconactor
        (:Is%20Valid
          (Class|BWeaponBunkerMapIconActor|ActivateOpened _mapiconactor))))))

(fn SetMapIconOpening
  ;[0.1161 0.1562 0.1419] SetMapIconOpening: Create the map icon actor, if it doesn't exist, and set the map icon.
  (Networking|SwitchHasAuthority
    (:Authority
      (bind _mapiconactor (CallFunction|CreateMapIcon))
      (Utilities|IsValid _mapiconactor
        (:Is%20Valid
          (Class|BWeaponBunkerMapIconActor|ActivateOpening _mapiconactor))))))

(fn SetWeaponBunkerDoorState
  (params (NewState :type WeaponBunkerDoorOpeningStateEnum%20Enum :default "NewEnumerator0"))
  (bind _weaponbunkerdoorstate (Variables|Transient|GetWeaponBunkerDoorState))
  ;[0.02494 0.105 0.1094] SetWeaponBunkerDoorState: Update the state of the weapon bunker door.
  (Networking|SwitchHasAuthority
    (:Authority
      (if (Utilities|Enum|NotEqual(Enum) NewState _weaponbunkerdoorstate)
        (Networking|FlushNetDormancy)
        (Fortnite|FortShippingLog self (Utilities|String|Append "SetWeaponBunkerDoorState: Switching state from " (Utilities|String|EnumtoString _weaponbunkerdoorstate) " to " (Utilities|String|EnumtoString NewState)))
        (bind _output_get (Variables|Transient|SetWeaponBunkerDoorState NewState))
        (CallFunction|ClientWeaponBunkerDoorStateChanged _output_get) ; Useful for handling burst events.
      ))))

(fn ConstructionScript
  (|CallParentFunction))
