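; Rank-gated scanner lock for the military weapon bunker door.
; Interacting starts a timed, re-entry-guarded scan where LockCheck maps the pawn's hero rank gameplay tag to an enum and compares it to a minimum.
; ScanEnum's OnRep drives the scanner ray's visibility, pass/fail color via custom primitive data, and sound cues.
; After a confirmation delay it fires a result cue, calls an empty quest-event hook, then defers to parent interaction.
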(event EventBeginPlay
  (|CallParentFunction)
  (Class|FortAsyncActionCurrentPlaylistReady|PlaylistDataReadyAsync
    (:OnReady
      (CallFunction|InitializeScannerRay)
      (Variables|Default|SetScanEnum "NewEnumerator0")
      (Networking|SwitchHasAuthority
        (:Authority
          (Variables|Default|SetOwningActor (Actor|GetParentActor)))))))

(event Interaction|EventBlueprintOnInteract
  (params (InteractingPawn :type Fort%20Pawn%20Object%20Reference) (InteractionBeingAttempted :type EInteractionBeingAttempted%20Enum :default "FirstInteraction"))
  (Variables|Default|GetbScanInProgress
    (:then)
    (:else
      (bind _checkpassed (CallFunction|LockCheck InteractingPawn InteractionBeingAttempted))
      (Variables|Interaction|SetbAllowInteract (not _checkpassed))
      (GameplayCue|ExecuteGameplayCueOnActor(Burst) self (Variables|Default|GetScanning%20Gameplay%20Cue%20Tag) (Ability|GameplayCue|MakeGameplayCueParameters 0.0))
      (Development|PrintString "INTERACT ATTEMPTED !!!" false false)
      (Variables|Default|SetbScanInProgress true)
      (Variables|Default|SetScanEnum "NewEnumerator1")
      (Utilities|FlowControl|Delay (ScalableFloat|GetValueatLevel (Variables|Config|GetInteractDuration)))
      (bind _checkpassed_2 (CallFunction|LockCheck InteractingPawn InteractionBeingAttempted))
      (Variables|Default|SetbLockCheckSucceeded _checkpassed_2)
      (Variables|Default|SetScanEnum "NewEnumerator2")
      (Utilities|FlowControl|Delay (ScalableFloat|GetValueatLevel (Variables|Config|GetInteractConfirmationDelay)))
      (GameplayCue|ExecuteGameplayCueOnActor(Burst) self (Variables|Default|GetInteractWithLockGameplay%20Cue%20Tag) (Ability|GameplayCue|MakeGameplayCueParameters (Math|Float|SelectFloat 1.0 0.0 (Variables|Default|GetbLockCheckSucceeded))))
      (if (not (Variables|Default|GetbLockCheckSucceeded))
        (Variables|Default|SetScanEnum "NewEnumerator0"))
      ;[0 0.0762 0.8854 0.5] Send Quest tag to interacting player on open success
      (CallFunction|SendQuestEventonDoorLockSuccess InteractingPawn)
      (Variables|Default|SetbScanInProgress false)
      (|CallParentFunction InteractingPawn InteractionBeingAttempted))))

(event Custom|BeginChannelAnim
  (params (PawnToAnimate :type Pawn%20Object%20Reference))
  (Variables|Default|SetChannelingPawn PawnToAnimate))

(event EventUpdateColors
  (params (State :type Enum_KeysAndLocks_LockState%20Enum :default "NewEnumerator0")))

(event EventOnUnlocked
  (|CallParentFunction)
  (Variables|Default|SetScanEnum "NewEnumerator3")
  (if (ScalableFloat|GetValueasBool (Variables|Default|GetShouldForceDisableInteract))
    (Variables|/Script/FortniteGame.BuildingActor|SetbAllowInteract false (Variables|LockParent|GetAttachedLockActor))))

(fn CheckRankRequirement
  (params (HeroRankToCheck :type Gameplay%20Tag%20Structure))
  (bind (_value _returnvalue) (Utilities|Map|Find (Variables|Default|GetHeroRankEnumTagMap) HeroRankToCheck))
  (return (and (<= (Math|Conversions|ToInteger(Byte) (Variables|Default|GetHeroRankMinimumRequirement)) (Math|Conversions|ToInteger(Byte) _value)) _returnvalue)))

(fn DelayedSetVisibilityAndColor
  (CallFunction|SetScannerColor (Variables|Default|GetScannerRay) (Variables|Scanner|GetScannerColorDefault))
  (Rendering|SetVisibility (Variables|Default|GetScannerRay) false true))

(fn InitializeScannerRay
  (Rendering|Material|SetCustomPrimitiveDataFloat (Variables|Default|GetScannerRay) 0 (ScalableFloat|GetValueatLevel (Variables|Scanner|GetScannerAngle))) ; Angle
  (Rendering|Material|SetCustomPrimitiveDataFloat (Variables|Default|GetScannerRay) 1 (ScalableFloat|GetValueatLevel (Variables|Scanner|GetScannerEndWidth))) ; WidthEnd
  (Rendering|Material|SetCustomPrimitiveDataFloat (Variables|Default|GetScannerRay) 2 (ScalableFloat|GetValueatLevel (Variables|Scanner|GetScannerBaseWidth))) ; WidthBase
  (Rendering|Material|SetCustomPrimitiveDataFloat (Variables|Default|GetScannerRay) 3 (ScalableFloat|GetValueatLevel (Variables|Scanner|GetScannerFalloffHardness))) ; FalloffHardness
  (Rendering|Material|SetCustomPrimitiveDataFloat (Variables|Default|GetScannerRay) 12 (ScalableFloat|GetValueatLevel (Variables|Scanner|GetScannerSecondaryMovementSpeed))) ; SecondaryMovementSpeed
  (Rendering|Material|SetCustomPrimitiveDataFloat (Variables|Default|GetScannerRay) 13 (ScalableFloat|GetValueatLevel (Variables|Scanner|GetScannerSecondaryFanOpacity))) ; SecondaryFanOpacity
  (CallFunction|SetScannerColor (Variables|Default|GetScannerRay) (Variables|Scanner|GetScannerColorDefault))
  (Rendering|Material|SetScalarParameterValueonMaterials (Variables|Default|GetScannerRay) "ScanRotAnim" (ScalableFloat|GetValueatLevel (Variables|Default|GetScannerAnimRate))))

(fn LerpPlayerHeight
  (while (<= (LocalVariables|GetLerpAlpha) 1.0)
    (Transformation|AddActorLocalOffset (Variables|Default|GetChannelingPawn) (Math|Vector|MakeVector 0.0 0.0 (Math|Float|Lerp 0.0 25.0 (LocalVariables|GetLerpAlpha))) false false)
    (LocalVariables|SetLerpAlpha (+ (LocalVariables|GetLerpAlpha) 0.1))))

(fn LockCheck
  (params (InteractingPawn :type Fort%20Pawn%20Object%20Reference) (InteractionBeingAttempted :type EInteractionBeingAttempted%20Enum :default "FirstInteraction"))
  (bind _currentheroranktag (LocalVariables|GetCurrentHeroRankTag))
  (bind _herorankenumtagmap (Variables|Default|GetHeroRankEnumTagMap))
  (bind (_value _returnvalue) (Utilities|Map|Find _herorankenumtagmap _currentheroranktag))
  (bind _keys (Utilities|Map|Keys (Variables|Default|GetHeroRankEnumTagMap)))
  (bind _array_element (Utilities|Array|ForEachLoopwithBreak _keys)
    (:LoopBody
      (if (GameplayTags|HasMatchingGameplayTag (Ability|GetAbilitySystemComponent InteractingPawn) _array_element)
        (LocalVariables|SetCurrentHeroRankTag _array_element)))
    (:Completed
      _))
  (if _returnvalue
    (return (<= (Math|Conversions|ToInteger(Byte) (Variables|Default|GetHeroRankMinimumRequirement)) (Math|Conversions|ToInteger(Byte) _value)))
    (else
      (return false))))

(fn OnRep_ScanEnum
  (switch Utilities|FlowControl|Switch|SwitchonWeaponBunkerLockStateEnum (Variables|Default|GetScanEnum)
    (:NewEnumerator0
      (Rendering|SetVisibility (Variables|Default|GetScannerRay) false true)
      (CallFunction|SetScannerColor (Variables|Default|GetScannerRay) (Variables|Scanner|GetScannerColorDefault)))
    (:NewEnumerator1
      (Rendering|SetVisibility (Variables|Default|GetScannerRay) true true))
    (:NewEnumerator2
      (CallFunction|SetScannerColor (Variables|Default|GetScannerRay) (Math|Color|SelectColor (Variables|Scanner|GetScannerColorSuccess) (Variables|Scanner|GetScannerColorFailed) (Variables|Default|GetbLockCheckSucceeded)))
      ;[0.09387 0.02533 0.15 0.5] Audio for Success or Fail Scan
      (if (Variables|Default|GetbLockCheckSucceeded)
        (GameplayCue|ExecuteGameplayCueOnActor(Burst) self (Variables|Default|GetLock%20Succeed%20Sound) (Ability|GameplayCue|MakeGameplayCueParameters 0.0))
        (else
          (GameplayCue|ExecuteGameplayCueOnActor(Burst) self (Variables|Default|GetLock%20Fail%20Sound) (Ability|GameplayCue|MakeGameplayCueParameters 0.0)))))
    (:NewEnumerator3
      ; Added to fix FORT-934133 to fully hide Scanner VFX
      (Utilities|Time|SetTimerbyFunctionName 0 "DelayedSetVisibilityAndColor" (ScalableFloat|GetValueatLevel (Variables|Config|GetDelayedSetVisibilityAndColorDuration)))
      (Rendering|SetVisibility (Variables|Default|GetScannerRay) false true))))

(fn SendQuestEventOnDoorLockSuccess
  (params (Actor :type Actor%20Object%20Reference)))

(fn SetHeroRankEnum
  (params (HeroRankToSet :type WeaponBunkerRankEnum%20Enum :default "NewEnumerator0"))
  (Variables|Default|SetHeroRankMinimumRequirement HeroRankToSet))

(fn SetScannerColor
  (params (ScannerRayStaticMesh :type Static%20Mesh%20Component%20Object%20Reference) (ScannerColor :type Linear%20Color%20Structure))
  (bind (_r _g _b _a) (Utilities|Struct|BreakLinearColor ScannerColor))
  (bind (_r_2 _g_2 _b_2 _a_2) (Utilities|Struct|BreakLinearColor ScannerColor))
  (Rendering|Material|SetCustomPrimitiveDataVector4 ScannerRayStaticMesh 4 (Math|Vector4|MakeVector4 _r_2 _g_2 _b_2 _a_2)) ; Color
  (Rendering|Material|SetCustomPrimitiveDataVector4 ScannerRayStaticMesh 8 (Math|Vector4|MakeVector4 _r _g _b _a)) ; Color
)

(fn UpdateVisualsBasedOnCurrentLockState
  (|CallParentFunction))

(fn ConstructionScript
  (|CallParentFunction))
